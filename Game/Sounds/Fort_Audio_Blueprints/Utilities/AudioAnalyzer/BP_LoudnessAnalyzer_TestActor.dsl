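; Test actor for measuring audio loudness with AudioSynesthesia analyzers and showing results in an on-screen MVVM widget.
; BeginPlay routes submixes to audio buses via a map, then creates one analyzer and row viewmodel per unique bus.
; Widget buttons start or stop analysis and a test tone, while handlers convert dB and semitones to volume and pitch.
; Loudness results are range-checked, though StopAnalyzingBusses starts inactive buses despite its "Stop AudioBus" label.
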(event Custom|OnPlayTestToneClicked
  ;[0 0 0 0.5] Test Tone
  (Audio|Components|Audio|Play (Variables|Default|GetAudio) 0.0))

(event Custom|StartAnalyzing
  ;[0 0.6042 0.08464 0.1913] StartAnalyzing
  (Networking|CanExecuteCosmeticEvents
    (:True
      (LoudnessAnalyzer|StartAnalyzingBusses))))

(event Custom|OnOverallLoudnessResults
  (params (LatestOverallLoudnessResults :type Loudness%20Results%20Structure%20%28by%20ref%29))
  ;[0 0.09759 0.5841 0.3217] Loudness Results Changed
  (CallFunction|RunLoudnessTest LatestOverallLoudnessResults))

(event EventBeginPlay
  ;[0.006408 0.2344 0 0.1913] BeginPlay
  (CallFunction|Init)
  (CallFunction|BindEvents)
  (Default|CallOnInit))

(event EventEndPlay
  (params (EndPlayReason :type EEndPlayReason%20Enum :default "Destroyed"))
  ;[1 0.4034 0.001437 0.113] EndPlay - Stop analyzer, or else ,':-/
  (CallFunction|StopAnalyzing))

(event Custom|StopTestToneClicked
  ;[0 0 0 0.5] Test Tone
  (Audio|Components|Audio|Stop (Variables|Default|GetAudio)))

(event Custom|StopAnalyzing
  ;[0.9062 0 0.07722 0.4] StopAnalyzing
  (Networking|CanExecuteCosmeticEvents
    (:True
      (LoudnessAnalyzer|StopAnalyzingBusses))))

(event Custom|OnGainValueChanged
  (params (Value :type Float%20%28single-precision%29))
  ;[0 0.09759 0.5841 0.3217] Gain Value Changed
  (CallFunction|SetComponentGain Value))

(event Custom|OnWidgetPitchValueChanged
  (params (Value :type Float%20%28single-precision%29))
  ;[0 0.09759 0.5841 0.3217] Pitch Value Changed
  (CallFunction|SetComponentPitch Value))

(event Custom|BindEvents
  (bind _mvvm_loudnessanalyzer (Variables|UI|GetMVVM_LoudnessAnalyzer))
  ;[0.1266 0.4844 0.1516 0.1913] Event Bindings
  ;;[0 0 0 0.5] Viewmodel Button Bindings
  (Utilities|IsValid (Variables|UI|GetMVVM_LoudnessAnalyzer)
    (:Is%20Valid
      (Default|BindEventtoOnStartClicked _mvvm_loudnessanalyzer (EventDispatchers|CreateEvent :!event Custom|StartClicked))
      (Default|BindEventtoOnStopClicked _mvvm_loudnessanalyzer (EventDispatchers|CreateEvent :!event Custom|StopClicked))
      (Default|BindEventtoOnPauseClicked _mvvm_loudnessanalyzer (EventDispatchers|CreateEvent :!event Custom|PauseClicked))
      (Default|BindEventtoOnPlayTestToneClicked _mvvm_loudnessanalyzer (EventDispatchers|CreateEvent :!event Custom|OnPlayTestToneClicked))
      (Default|BindEventtoOnStopTestToneClicked _mvvm_loudnessanalyzer (EventDispatchers|CreateEvent :!event Custom|StopTestToneClicked)))))

(event Custom|StartClicked
  ;[0 0.09759 0.5841 0.3217] Button Events
  (CallFunction|StartAnalyzing))

(event Custom|StopClicked
  ;[0 0.09759 0.5841 0.3217] Button Events
  (CallFunction|StopAnalyzing))

(event Custom|PauseClicked
  ;[0 0.09759 0.5841 0.3217] Button Events
  (CallFunction|StopAnalyzing))

(fn BroadcastLoudnessResults
  (params (Loudness%20Results :type Loudness%20Results%20Structure))
  ;[0 0 0 0.5] Broadcast loudness to all scrollbox viewmodels, as well as the viewmodel for the Main widget
  (for _array_element (Variables|/Game/Sounds/Fort_Audio_Blueprints/Utilities/AudioAnalyzer/LoudnessAnalyzer/Viewmodels/MVVM_LoudnessAnalyzer_AudioBusScrollBox.MVVM_LoudnessAnalyzer_AudioBusScrollBox_C|GetViewmodels (Variables|UI|GetMVVM%20Loudness%20Analyzer%20Audio%20Bus%20Scroll%20Box))
    (Variables|/Game/Sounds/Fort_Audio_Blueprints/Utilities/AudioAnalyzer/LoudnessAnalyzer/Viewmodels/MVVM_LoudnessAnalyzer_AudioBusScrollBoxItem.MVVM_LoudnessAnalyzer_AudioBusScrollBoxItem_C|SetLoudnessResults Loudness%20Results _array_element))
  (Variables|/Game/Sounds/Fort_Audio_Blueprints/Utilities/AudioAnalyzer/LoudnessAnalyzer/Viewmodels/MVVM_LoudnessAnalyzer.MVVM_LoudnessAnalyzer_C|SetLoudnessResults Loudness%20Results (Variables|UI|GetMVVM_LoudnessAnalyzer)))

(fn BroadcastLoudnessResults_0
  (params (Loudness%20Results :type Loudness%20Results%20Structure) (LoudnessAnalyzer :type Loudness%20Analyzer%20Object%20Reference%20%28by%20ref%29))
  ;[0 0 0 0.5] Broadcast to viewmodel associated with input LoudnessAnalyzer
  (Variables|/Game/Sounds/Fort_Audio_Blueprints/Utilities/AudioAnalyzer/LoudnessAnalyzer/Viewmodels/MVVM_LoudnessAnalyzer_AudioBusScrollBoxItem.MVVM_LoudnessAnalyzer_AudioBusScrollBoxItem_C|SetLoudnessResults Loudness%20Results (Utilities|Map|Find (Variables|LoudnessAnalyzer|GetAnalyzerToViewmodelMap) LoudnessAnalyzer)))

(fn BroadcastTestResults
  (params (TestResults :type S%20Loudness%20Analyzer%20Test%20Results%20Structure))
  (bind (_keypin _valuepin) (Utilities|Map|ForEachLoop(Map) (Variables|LoudnessAnalyzer|GetAnalyzerToViewmodelMap)))
  (Variables|/Game/Sounds/Fort_Audio_Blueprints/Utilities/AudioAnalyzer/LoudnessAnalyzer/Viewmodels/MVVM_LoudnessAnalyzer_AudioBusScrollBoxItem.MVVM_LoudnessAnalyzer_AudioBusScrollBoxItem_C|SetTestResults TestResults _valuepin))

(fn CanDisplayEasyInteractButton
  (return false))

(fn CanDisplaySecondaryEasyInteractButton
  (return false))

(fn CanMaintainLongInteractionWhileMoving
  (return false))

(fn ConstructLoudnessAnalyzer
  ;[0 0 0 0.5] Create LoudnessAnalyzer object
  (bind _returnvalue (Game|ConstructObjectfromClass "/Game/Sounds/Fort_Audio_Blueprints/Utilities/AudioAnalyzer/BP_LoudnessAnalyzer_Settings.BP_LoudnessAnalyzer_Settings_C"))
  (bind _output_get (Variables|LoudnessAnalyzer|SetLoudnessAnalyzerSettings _returnvalue))
  (bind _returnvalue_2 (Game|ConstructObjectfromClass "/Script/AudioSynesthesia.LoudnessAnalyzer" 0 _output_get))
  (Variables|LoudnessAnalyzer|SetLoudnessAnalyzer _returnvalue_2))

(fn ConstructLoudnessAnalyzers
  ;[0 0 0 0.5] Create LoudnessAnalyzer for each unique AudioBus entered by the user
  (for _array_element (Variables|LoudnessAnalyzer|GetAudioBusses)
    (bind _returnvalue (Game|ConstructObjectfromClass "/Game/Sounds/Fort_Audio_Blueprints/Utilities/AudioAnalyzer/BP_LoudnessAnalyzer_Settings.BP_LoudnessAnalyzer_Settings_C"))
    (bind _returnvalue_2 (Game|ConstructObjectfromClass "/Script/AudioSynesthesia.LoudnessAnalyzer" 0 _returnvalue))
    (Utilities|Map|Add (Variables|LoudnessAnalyzer|GetLoudnessAnalyzers) _array_element _returnvalue_2)))

(fn ConstructViewmodels
  ;[0 0.3916 0.8879 0.487] Create new viewmodel for each AudioBus found in the SubmixRoutingMap
  (bind _values (Utilities|Map|Values (Variables|LoudnessAnalyzer|GetLoudnessAnalyzers)))
  (for _array_element _values
    (bind _output_get (LocalVariables|SetCurrentLoudnessAnalyzer _array_element))
    (Utilities|IsValid _output_get
      (:Is%20Valid
        (bind _returnvalue (Game|ConstructObjectfromClass "/Game/Sounds/Fort_Audio_Blueprints/Utilities/AudioAnalyzer/LoudnessAnalyzer/Viewmodels/MVVM_LoudnessAnalyzer_AudioBusScrollBoxItem.MVVM_LoudnessAnalyzer_AudioBusScrollBoxItem_C"))
        (bind _output_get_2 (LocalVariables|SetNewViewmodel _returnvalue))
        (Class|MVVMLoudnessAnalyzerAudioBusScrollBoxItem|SetLoudnessAnalyzer _output_get_2 (LocalVariables|GetCurrentLoudnessAnalyzer) :!class /Game/Sounds/Fort_Audio_Blueprints/Utilities/AudioAnalyzer/LoudnessAnalyzer/Viewmodels/MVVM_LoudnessAnalyzer_AudioBusScrollBoxItem.MVVM_LoudnessAnalyzer_AudioBusScrollBoxItem_C) ; Set analyzer reference in order to bind to OnLatestestOverallLoudnessResults at a per-viewmodel level
        (Utilities|Map|Add (Variables|LoudnessAnalyzer|GetAnalyzerToViewmodelMap) (LocalVariables|GetCurrentLoudnessAnalyzer) (LocalVariables|GetNewViewmodel)))))
  ;[0.2232 0.01681 0.3278 0.287] Return viewmodels associated with an AudioBus
  ; The list of viewmodels is used to populate the UI scrollbox
  (bind _values_2 (Utilities|Map|Values (Variables|LoudnessAnalyzer|GetAnalyzerToViewmodelMap)))
  (return _values_2))

(fn CreateWidget
  (bind _loudnessanalyzerwidget (Variables|UI|GetLoudnessAnalyzerWidget))
  (Utilities|FlowControl|Sequence
    (:then_0
      ;[0 0 0 0.5] Create LoudnessMeter Widget
      ; Related viewmodels are marked 'CreateInstance', and thus constructed with the widget
      (bind _returnvalue (UserInterface|CreateWidget "/Game/Sounds/Fort_Audio_Blueprints/Utilities/AudioAnalyzer/LoudnessAnalyzer/Widgets/WB_LoudnessAnalyzer_Main.WB_LoudnessAnalyzer_Main_C" (Game|GetPlayerController 0) 0 0))
      (Variables|UI|SetLoudnessAnalyzerWidget _returnvalue))
    (:then_1
      ;[0 0 0 0.5] Store references to Viewmodels created by widget
      (Utilities|IsValid (Variables|UI|GetLoudnessAnalyzerWidget)
        (:Is%20Valid
          (Variables|UI|SetMVVM_LoudnessAnalyzer (Variables|/Game/Sounds/Fort_Audio_Blueprints/Utilities/AudioAnalyzer/LoudnessAnalyzer/Widgets/WB_LoudnessAnalyzer_Main.WB_LoudnessAnalyzer_Main_C|GetMVVM_AudioAnalyzerTest _loudnessanalyzerwidget))
          (Variables|UI|SetMVVM%20Loudness%20Analyzer%20Audio%20Bus%20Scroll%20Box (Variables|/Game/Sounds/Fort_Audio_Blueprints/Utilities/AudioAnalyzer/LoudnessAnalyzer/Widgets/WB_LoudnessAnalyzer_Main.WB_LoudnessAnalyzer_Main_C|GetMVVM_LoudnessAnalyzer_AudioBusScrollBox _loudnessanalyzerwidget)))))
    (:then_2
      ;[0 0 0 0.5] Add 'AudioBusScrollBoxItem' widgets to ScrollBox
      ; Create one viewmodel for each unique AudioBus in the SubmixRoutingMap
      ; One widget is added to the ScrollBox for each viewmodel created
      (bind _viewmodels (UI|ConstructViewmodels))
      (bind _mvvm_loudness_analyzer_audio_bus_scroll_box (Variables|UI|GetMVVM%20Loudness%20Analyzer%20Audio%20Bus%20Scroll%20Box)
        (:then
          (Class|MVVMLoudnessAnalyzerAudioBusScrollBox|SetItems _mvvm_loudness_analyzer_audio_bus_scroll_box _viewmodels))
        (:else)))
    (:then_4
      ;[0 0 0 0.5] Add LoudnessMeter widget to viewport
      (Utilities|IsValid (Variables|UI|GetLoudnessAnalyzerWidget)
        (:Is%20Valid
          (UserInterface|Viewport|AddtoViewport (Variables|UI|GetLoudnessAnalyzerWidget) 0)
          (CallFunction|SetScrollBoxItemDefaults))))
    (:then_5
      ;[0 0 0 0.5] Set Input Mode
      (Input|SetInputModeGameAndUI (Game|GetPlayerController 0) (Variables|UI|GetLoudnessAnalyzerWidget) "DoNotLock" false))))

(fn CreateWidgets
  (bind _loudnessanalyzerwidget (Variables|UI|GetLoudnessAnalyzerWidget))
  (Utilities|FlowControl|Sequence
    (:then_0
      ;[0 0 0 0.5] Create LoudnessMeter Widget
      ; Related viewmodels are marked 'CreateInstance', and thus constructed with the widget
      (bind _returnvalue (UserInterface|CreateWidget "/Game/Sounds/Fort_Audio_Blueprints/Utilities/AudioAnalyzer/LoudnessAnalyzer/Widgets/WB_LoudnessAnalyzer_Main.WB_LoudnessAnalyzer_Main_C" (Game|GetPlayerController 0) 0 0))
      (Variables|UI|SetLoudnessAnalyzerWidget _returnvalue))
    (:then_1
      ;[0 0 0 0.5] Store references to Viewmodels created by widget
      (Utilities|IsValid (Variables|UI|GetLoudnessAnalyzerWidget)
        (:Is%20Valid
          (Variables|UI|SetMVVM_LoudnessAnalyzer (Variables|/Game/Sounds/Fort_Audio_Blueprints/Utilities/AudioAnalyzer/LoudnessAnalyzer/Widgets/WB_LoudnessAnalyzer_Main.WB_LoudnessAnalyzer_Main_C|GetMVVM_AudioAnalyzerTest _loudnessanalyzerwidget))
          (Variables|UI|SetMVVM%20Loudness%20Analyzer%20Audio%20Bus%20Scroll%20Box (Variables|/Game/Sounds/Fort_Audio_Blueprints/Utilities/AudioAnalyzer/LoudnessAnalyzer/Widgets/WB_LoudnessAnalyzer_Main.WB_LoudnessAnalyzer_Main_C|GetMVVM_LoudnessAnalyzer_AudioBusScrollBox _loudnessanalyzerwidget)))))
    (:then_2
      ;[0 0 0 0.5] Add 'AudioBusScrollBoxItem' widgets to ScrollBox
      ; Create one viewmodel for each unique AudioBus in the SubmixRoutingMap
      ; One widget is added to the ScrollBox for each viewmodel created
      (bind _viewmodels (UI|ConstructViewmodels)) ; Called after widget creation to trigger ViewBinding updates on widget
      (bind _mvvm_loudness_analyzer_audio_bus_scroll_box (Variables|UI|GetMVVM%20Loudness%20Analyzer%20Audio%20Bus%20Scroll%20Box)
        (:then
          (Class|MVVMLoudnessAnalyzerAudioBusScrollBox|SetItems _mvvm_loudness_analyzer_audio_bus_scroll_box _viewmodels))
        (:else)))
    (:then_3
      ;[0 0 0 0.5] Add LoudnessMeter widget to viewport
      (Utilities|IsValid (Variables|UI|GetLoudnessAnalyzerWidget)
        (:Is%20Valid
          (UserInterface|Viewport|AddtoViewport (Variables|UI|GetLoudnessAnalyzerWidget) 0)
          (CallFunction|SetScrollBoxItemDefaults))))
    (:then_4
      ;[0 0 0 0.5] Set Input Mode
      (Input|SetInputModeGameAndUI (Game|GetPlayerController 0) (Variables|UI|GetLoudnessAnalyzerWidget) "DoNotLock" false))))

(fn GetActorForInteraction
  (return 0))

(fn GetBoundsComponents
  (params (SelfActor :type Actor%20Object%20Reference))
  (return 0))

(fn GetConnectedInteractableActors
  (params (ViewingActor :type Actor%20Object%20Reference))
  (return 0))

(fn GetConnectedVisibleActor
  (return 0))

(fn GetCustomIcon
  (return ""))

(fn GetCustomTint
  (return 0))

(fn GetFocusedSocketLocation
  (return "0, 0, 0"))

(fn GetIconPlacement
  (params (SelfActor :type Actor%20Object%20Reference) (ViewingActor :type Actor%20Object%20Reference))
  (return "0, 0, 0" "0, 0, 0"))

(fn GetInteractionEventPayload
  (params (InteractType :type Interaction%20Type%20Structure%20%28by%20ref%29) (InteractingPawn :type Fort%20Pawn%20Object%20Reference))
  (return 0 false))

(fn GetMinDistanceFromInteraction
  (return 0.0))

(fn GetOverrideItemInteract
  (params (InteractType :type Interaction%20Type%20Structure%20%28by%20ref%29))
  (return 0 false))

(fn GetRoutedSubmixes
  (params (AudioBus :type Audio%20Bus%20Object%20Reference))
  (LocalVariables|SetComparisonBus AudioBus)
  ;[0 0.3916 0.8879 0.487] Map entries whose value matches the ComparisonBus should return their key: the SoundSubmix asset reference
  (bind (_keypin _valuepin) (Utilities|Map|ForEachLoop(Map) (Variables|LoudnessAnalyzer|GetSubmixRoutingMap))
    (:then
      (if (== _valuepin (LocalVariables|GetComparisonBus))
        (Utilities|Array|AddUnique (LocalVariables|GetOutRoutedSubmixes) _keypin)))
    (:CompletedPin
      ;[0.2232 0.01681 0.3278 0.287] Return Submixes mapped to this AudioBus
      (return (LocalVariables|GetOutRoutedSubmixes)))))

(fn GetRoutedSubmixNames
  (params (AudioBus :type Audio%20Bus%20Object%20Reference))
  ;[0 0.3968 0.8963 0.2522] Get submixes mapped to the input AudioBus
  (bind _routedsubmixes (CallFunction|GetRoutedSubmixes AudioBus))
  (for (_array_element _array_index) _routedsubmixes
    ;[0.9823 0 0.6445 0.2696] Concatenate submix names
    (LocalVariables|SetOutSubmixNames (Utilities|String|Append (LocalVariables|GetOutSubmixNames) (Utilities|GetObjectName _array_element) (Utilities|String|SelectString "" ", " (== _array_index (Utilities|Array|LastIndex _routedsubmixes))))))
  ;[0.2232 0.01681 0.3278 0.287] Return comma-separated list of submixes routed to the input AudioBus
  (return (LocalVariables|GetOutSubmixNames)))

(fn GetSecondaryCustomIcon
  (return ""))

(fn GetSecondaryIconPlacement
  (params (SelfActor :type Actor%20Object%20Reference) (ViewingActor :type Actor%20Object%20Reference))
  (return "0, 0, 0" "0, 0, 0"))

(fn IconPlacementNeedsUpdate
  (return false))

(fn Init
  (LoudnessAnalyzer|RegisterAudioBussestoSoundSubmixes) ; Populate AudioBusses array
  (LoudnessAnalyzer|ConstructLoudnessAnalyzers) ; Create the LoudnessAnalyzer objects which will monitor each AudioBus
  (UI|CreateWidgets) ; Create widget with scrollboxes containing each AudioBus
)

(fn LocalOnFailedInteract
  (params (InteractingPawn :type Fort%20Player%20Pawn%20Object%20Reference)))

(fn PrintCurveType
  (bind _loudnessanalyzer (Variables|LoudnessAnalyzer|GetLoudnessAnalyzer)
    (:then
      (bind _settings (Variables|/Script/AudioSynesthesia.LoudnessAnalyzer|GetSettings _loudnessanalyzer)
        (:then
          (Development|PrintText (Utilities|Text|FormatText "NSLOCTEXT(\"[AF40B40231757242D7F12C3645EF407E]\", \"D8122F304CF30D1AAC4DEA9834848EE5\", \"Curve Type: {c}\")" (Utilities|String|EnumtoString (Variables|/Script/AudioSynesthesia.LoudnessSettings|GetCurveType _settings)))))
        (:else)))
    (:else)))

(fn RegisterAudioBussesToSoundSubmixes
  ;[0 0 0 0.5] Route submixes to respective audio busses
  (bind (_keypin _valuepin) (Utilities|Map|ForEachLoop(Map) (Variables|LoudnessAnalyzer|GetSubmixRoutingMap))
    (:then
      (Audio|Bus|RegisterAudioBustoSubmix _keypin _valuepin))
    (:CompletedPin
      ;[0 0 0 0.5] Store list of unique AudioBusses from the map for convenient iteration
      (bind _values (Utilities|Map|Values (Variables|LoudnessAnalyzer|GetSubmixRoutingMap)))
      (for _array_element _values
        (Utilities|Array|AddUnique (Variables|LoudnessAnalyzer|GetAudioBusses) _array_element)))))

(fn RunLoudnessTest
  (params (LoudnessResults :type Loudness%20Results%20Structure%20%28by%20ref%29))
  (bind _testresults (Validation|ValidateLoudness LoudnessResults))
  (CallFunction|BroadcastTestResults _testresults))

(fn SetComponentGain
  (params (dB :type Float%20%28single-precision%29))
  (Audio|Components|Audio|SetVolumeMultiplier (Variables|Default|GetAudio) (Math|Float|Power 1e+01 (/ dB 2e+01)) :!class /Script/Engine.AudioComponent))

(fn SetComponentPitch
  (params (NewPitchOffsetSemitones :type Float%20%28single-precision%29))
  (Audio|Components|Audio|SetPitchMultiplier (Variables|Default|GetAudio) (Math|Float|Power 2.0 (/ NewPitchOffsetSemitones 12.0))))

(fn SetScrollBoxItemDefaults
  (bind _analyzertoviewmodelmap (Variables|LoudnessAnalyzer|GetAnalyzerToViewmodelMap))
  (bind (_keypin _valuepin) (Utilities|Map|ForEachLoop(Map) (Variables|LoudnessAnalyzer|GetLoudnessAnalyzers)))
  (if (.ReturnValue (Utilities|Map|Find _analyzertoviewmodelmap _valuepin))
    ;[0 0.003479 0.3854 0.287] Set ScrollBoxItem defaults
    (bind _submixnames (CallFunction|GetRoutedSubmixNames _keypin)) ; Return comma-separated list of submixes routed to the input AudioBus
    (Variables|/Game/Sounds/Fort_Audio_Blueprints/Utilities/AudioAnalyzer/LoudnessAnalyzer/Viewmodels/MVVM_LoudnessAnalyzer_AudioBusScrollBoxItem.MVVM_LoudnessAnalyzer_AudioBusScrollBoxItem_C|SetAudioBusRowSettings_AudioBusText_41_C7E76C2A44D34CF7DC331C8582E01395 (Utilities|Struct|MakeSLoudnessAnalyzerAudioBusRowSettings (Utilities|Text|ToText(String) (Utilities|GetObjectName _keypin)) (Utilities|Text|FormatText "NSLOCTEXT(\"[AF40B40231757242D7F12C3645EF407E]\", \"400AD0AC4BD31621E6EC8F93621FB045\", \"< {names} >\")" _submixnames)) (Utilities|Map|Find _analyzertoviewmodelmap _valuepin))))

(fn ShouldDeferAbilityEndToAbility
  (return false))

(fn StartAnalyzingBusses
  (bind _currentaudiobus (LocalVariables|GetCurrentAudioBus))
  ;[0 0.3916 0.8879 0.487] Iterate over each AudioBus/LoudnessAnalyzer pair
  (bind (_keypin _valuepin) (Utilities|Map|ForEachLoop(Map) (Variables|LoudnessAnalyzer|GetLoudnessAnalyzers)))
  (LocalVariables|SetCurrentAudioBus _keypin)
  (LocalVariables|SetCurrentLoudnessAnalyzer _valuepin)
  (Utilities|FlowControl|Sequence
    (:then_0
      ;[0 0.651 0.1245 0.4522] Start AudioBus
      (Utilities|IsBusActive _currentaudiobus
        (:OnBusInactive
          (Audio|Bus|StartAudioBus _currentaudiobus))))
    (:then_1
      ;[1 0.1779 0.8784 0.4522] Start LoudnessAnalyzer
      (AudioAnalyzer|StartAnalyzing (LocalVariables|GetCurrentLoudnessAnalyzer) (LocalVariables|GetCurrentAudioBus)))))

(fn StopAnalyzingBusses
  (bind _currentaudiobus (LocalVariables|GetCurrentAudioBus))
  ;[0 0.3916 0.8879 0.487] Iterate over each AudioBus/LoudnessAnalyzer pair
  (bind (_keypin _valuepin) (Utilities|Map|ForEachLoop(Map) (Variables|LoudnessAnalyzer|GetLoudnessAnalyzers)))
  (LocalVariables|SetCurrentAudioBus _keypin)
  (LocalVariables|SetCurrentLoudnessAnalyzer _valuepin)
  (Utilities|FlowControl|Sequence
    (:then_0
      ;[1 0.1779 0.8784 0.4522] Stop LoudnessAnalyzer
      (AudioAnalyzer|StopAnalyzing (LocalVariables|GetCurrentLoudnessAnalyzer)))
    (:then_1
      ;[1 0.06364 0.12 0.7217] Stop AudioBus
      (Utilities|IsBusActive _currentaudiobus
        (:OnBusInactive
          (Audio|Bus|StartAudioBus _currentaudiobus))))))

(fn ConstructionScript)

(fn ValidateLoudness
  (params (LoudnessResults :type Loudness%20Results%20Structure%20%28by%20ref%29))
  (LocalVariables|SetResults LoudnessResults)
  ;[0 0 0 0.5] Does Loudness fall within acceptable range?
  (LocalVariables|SetLoudnessTestSuccess (Math|Float|InRange(Float) (Utilities|Struct|BreakLoudnessResults (LocalVariables|GetResults)) (Variables|Validation|GetMin%20Loudness) (Variables|Validation|GetMax%20Loudness)))
  ;[0 0 0 0.5] Loudness Deviation
  ; (0 if test was successful)
  (LocalVariables|SetLoudnessDeviation (Math|Float|SelectFloat 0.0 (Math|Float|SelectFloat 1.0 -1.0 (> (Utilities|Struct|BreakLoudnessResults (LocalVariables|GetResults)) (Variables|Validation|GetMax%20Loudness))) (LocalVariables|GetLoudnessTestSuccess)))
  ;[0.2232 0.01681 0.3278 0.287] Return test results
  (return (Utilities|Struct|MakeSLoudnessAnalyzerTestResults (LocalVariables|GetLoudnessTestSuccess) (LocalVariables|GetLoudnessDeviation))))
